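# Physics asset for the Teriyaki Fish Princess backpack cosmetic, defining bodies and joint limits for its wings, ribbons and roses.
# Four box-shaped wings and three near-spherical capsule roses each constrain directly to the pack body.
# Two capsule ribbon chains hang from the pack, four links on the left and three on the right, loosening down the chain.
# Wings swing widely on one axis only, while roses stay tightly limited.
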PhysicsAsset /BRCosmetics/Accessories/FORT_Backpacks/Backpack_M_MED_Teriyaki_Fish_Princess/Meshes/M_MED_Teriyaki_Fish_Princess_Pack_Physics1
bodies {
 dyn_ribbon01_l [{shapeType=Capsule, length=1.01, radius=0.75, center=<0.6581 0 0>, rotation=<rot 0 90 90>}]
 dyn_ribbon01_r [{shapeType=Capsule, length=1.01, radius=0.9215, center=<0.5119 0 0.2199>, rotation=<rot 0 90 90>}]
 dyn_ribbon02_l [{shapeType=Capsule, length=1.26, radius=0.6796, center=<0.4175 0 0>, rotation=<rot 0 90 90>}]
 dyn_ribbon02_r [{shapeType=Capsule, length=1.01, radius=0.9215, center=<0.3242 0 0.3159>, rotation=<rot 0 90 90>}]
 dyn_ribbon03_l [{shapeType=Capsule, length=1.2203, radius=0.6863, center=<0.4605 0 0>, rotation=<rot 0 90 90>}]
 dyn_ribbon03_r [{shapeType=Capsule, length=1.01, radius=0.9215, center=<0.2546 0.0001 0.319>, rotation=<rot 0 90 90>}]
 dyn_ribbon04_l [{shapeType=Capsule, length=1.3855, radius=0.6217, center=<0.251 0 0>, rotation=<rot 0 90 90>}]
 dyn_rose1 [{shapeType=Capsule, length=0.005, radius=1.005, center=<0 -0.0399 0.3388>, rotation=<rot 0 -90 0>}]
 dyn_rose2 [{shapeType=Capsule, length=0.005, radius=1.005, center=<-0.455 -0.1061 -0.3402>, rotation=<rot 10.925 -75.0357 -125.3443>}]
 dyn_rose3 [{shapeType=Capsule, length=0.005, radius=1.005, center=<0.2272 -0.4563 0.0105>, rotation=<rot -2.2771 -149.5984 88.6333>}]
 dyn_wing_bottom_l [{shapeType=Box, extentX=1.364, extentY=7.6066, extentZ=5.2628, center=<-2.7667 0.8427 -0.1427>, rotation=<rot 4.8032 -90.8562 -0.4369>}]
 dyn_wing_bottom_r [{shapeType=Box, extentX=0.5, extentY=6.5756, extentZ=5.3441, center=<3.2895 -1.4067 0.423>, rotation=<rot 2.4345 -89.3542 -8.9225>}]
 dyn_wing_top_l [{shapeType=Box, extentX=1.51, extentY=8.01, extentZ=6.51, center=<-2.9749 -1.2259 -0.0001>, rotation=<rot 0 -90 0>}]
 dyn_wing_top_r [{shapeType=Box, extentX=8.01, extentY=2.01, extentZ=6.51, center=<2.3829 0.7962 0>}]
 pack [{shapeType=Capsule, length=0.4314, radius=1.0309, center=<0.0089 0.0183 -1.2166>, rotation=<rot 0.1321 -90 -1.0114>}]
}
constraints [
 {bone1Name=dyn_wing_top_l, bone2Name=pack, swing1LimitDegrees=45, swing2LimitDegrees=5, twistLimitDegrees=5}
 {bone1Name=dyn_wing_bottom_l, bone2Name=pack, swing1LimitDegrees=45, swing2LimitDegrees=10, twistLimitDegrees=5}
 {bone1Name=dyn_wing_top_r, bone2Name=pack, swing1LimitDegrees=45, swing2LimitDegrees=5, twistLimitDegrees=5}
 {bone1Name=dyn_wing_bottom_r, bone2Name=pack, swing1LimitDegrees=45, swing2LimitDegrees=10, twistLimitDegrees=5}
 {bone1Name=dyn_ribbon01_l, bone2Name=pack, swing1LimitDegrees=5, swing2LimitDegrees=10, twistLimitDegrees=10}
 {bone1Name=dyn_ribbon02_l, bone2Name=dyn_ribbon01_l, swing1LimitDegrees=15, swing2LimitDegrees=15, twistLimitDegrees=10}
 {bone1Name=dyn_ribbon03_l, bone2Name=dyn_ribbon02_l, swing1LimitDegrees=15, swing2LimitDegrees=20, twistLimitDegrees=10}
 {bone1Name=dyn_ribbon04_l, bone2Name=dyn_ribbon03_l, swing1LimitDegrees=15, swing2LimitDegrees=25, twistLimitDegrees=10}
 {bone1Name=dyn_ribbon01_r, bone2Name=pack, swing1LimitDegrees=5, swing2LimitDegrees=10, twistLimitDegrees=10}
 {bone1Name=dyn_ribbon02_r, bone2Name=dyn_ribbon01_r, swing1LimitDegrees=25, swing2LimitDegrees=15, twistLimitDegrees=10}
 {bone1Name=dyn_ribbon03_r, bone2Name=dyn_ribbon02_r, swing1LimitDegrees=25, swing2LimitDegrees=20, twistLimitDegrees=10}
 {bone1Name=dyn_rose1, bone2Name=pack, swing1LimitDegrees=10, swing2LimitDegrees=10, twistLimitDegrees=5}
 {bone1Name=dyn_rose2, bone2Name=pack, swing1LimitDegrees=10, swing2LimitDegrees=10, twistLimitDegrees=5}
 {bone1Name=dyn_rose3, bone2Name=pack, swing1LimitDegrees=10, swing2LimitDegrees=10, twistLimitDegrees=5}
]
