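# Animation Blueprint for the Fish Princess backpack, opening and closing its wings as the wearer moves.
# A state machine selects wing poses for jumping, falling, targeting, crouching and nested skydive modes, then two RigidBody passes add physics in pack-bone space.
# The cached result gets a speed-driven additive wing close, skipped while crouching or skydiving.
# The event graph derives that speed, zeroed during vertical motion, and an unused spine-based Backpack Rotate.
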AnimBlueprint /BRCosmetics/Accessories/FORT_Backpacks/Backpack_M_MED_Teriyaki_Fish_Princess/Meshes/M_MED_Teriyaki_Fish_Princess_Pack_AnimBP
target_skeleton=&/M_MED_Teriyaki_Fish_Princess_Pack_Skeleton
graphs [
 AnimGraph {nodes [
   n1 = Wing_Movement_Base_Anim_PoseAsset_Additive : Close_Additive!(PoseWeight:float=Get SprintSpeedRemap() -> (SprintSpeedRemap:double)) -> (Pose:Pose Link Structure) {
    is=title AnimGraphNode_PoseByName
    poseAsset=&/Wing_Movement_Base_Anim_PoseAsset_Additive
    poseName=Close_Additive
   }
   n2 = Wing_Movement_Base_Anim_PoseAsset_Additive : Close_Additive(PoseWeight=0.6) {
    poseAsset=&Wing_Movement_Base_Anim_PoseAsset_Additive
    poseName=Close_Additive
   }
   n3 = Get SprintSpeedRemap()
   n4 = Two Way Blend(
    A=Apply Additive(Base=Use cached pose 'SprintCache'(), Additive=@n2, Alpha=@n3) {alphaBoolBlend {blendOption=Linear}, alphaScaleBiasClamp {bInterpResult=true, interpSpeedDecreasing=5}}
    B=Apply Additive(Base=Use cached pose 'SprintCache'(), Additive=@n1, Alpha=0.25) {alphaBoolBlend {blendOption=Linear}}
    Alpha=MapRangeClamped(Value=@n3, InRangeA=0, InRangeB=1, OutRangeA=1, OutRangeB=0)
   ) {blendNode {alphaBoolBlend {blendOption=Linear}}}
   n5 = Use cached pose 'SprintCache'()
   n6 = Get bIsCrouching()
   Output Pose(Result=Blend Poses by bool(
     BlendPose_0=Use cached pose 'SprintCache'()
     BlendPose_1=Blend Poses by bool(BlendPose_0=@n5, BlendPose_1=@n4, BlendTime_1=0.3, bActiveValue=@n6) {blendType=Linear}
     bActiveValue=Get bIsSkyDiving()
    ) {blendType=Linear}) {layerGroup=None}
   AnimGraphNode_StateMachine_0 = wingsPosition() {is=title AnimGraphNode_StateMachine}
   n7 = RigidBody(ComponentPose=Local To Component(LocalPose=@AnimGraphNode_StateMachine_0)) {
    bClampLinearTranslationLimitToRefPose=true
    bOverrideWorldGravity=true
    overridePhysicsAsset=&/M_MED_Teriyaki_Fish_Princess_Pack_Physics
    simulationSpace=BaseBoneSpace
    alphaBoolBlend {blendOption=Linear}
    baseBoneRef {boneName=pack}
    componentAppliedLinearAccClamp=<1500 1500 1500>
    componentLinearAccScale=<0.6 0.25 0.35>
    componentLinearVelScale=<0.6 0.25 0.35>
    overrideWorldGravity {z=-500}
   }
   n8 = RigidBody(ComponentPose=@n7) {
    bClampLinearTranslationLimitToRefPose=true
    bOverrideWorldGravity=true
    overridePhysicsAsset=&/M_MED_Teriyaki_Fish_Princess_Pack_Physics1
    simulationSpace=BaseBoneSpace
    alphaBoolBlend {blendOption=Linear}
    baseBoneRef {boneName=pack}
    componentAppliedLinearAccClamp=<1100 900 900>
    componentLinearAccScale=<0.6 0.25 0.35>
    componentLinearVelScale=<0.6 0.25 0.35>
    overrideWorldGravity {z=-850}
   }
   n9 = "Transform (Modify) Bone - Bone: pack"(ComponentPose=@n8, Translation="0,-1.35,0", Rotation="0,0,0") {
    rotationMode=BMM_Additive
    rotationSpace=BCS_BoneSpace
    translationMode=BMM_Additive
    translationSpace=BCS_BoneSpace
    alphaBoolBlend {blendOption=Linear}
    boneToModify {boneName=pack}
    translation {y=-1.35}
   }
   Save cached pose 'SprintCache'(Pose=Component To Local(ComponentPose=@n9)) {cacheName=SprintCache}
  ]}
 wingsPosition {
  entry=Base
  path=AnimGraph.AnimGraphNode_StateMachine_0.wingsPosition
  ids=id
  nodes [
   AnimStateNode_0 = Base() {is=title AnimStateNode}
   AnimStateNode_1 = Jumping() {is=title AnimStateNode}
   AnimStateNode_2 = Falling() {is=title AnimStateNode}
   AnimStateNode_3 = Targeting() {is=title AnimStateNode}
   AnimStateNode_5 = SkydiveMode() {is=title AnimStateNode}
   AnimStateNode_9 = Crouching() {is=title AnimStateNode}
   AnimStateTransitionNode_0 = AnimStateTransitionNode() {blendMode=Custom, customBlendCurve=/Game/Accessories/FORT_Backpacks/M_MED_Demon_Pack/Animation/JumpTransitionBlendCurve, priorityOrder=2}
   AnimStateTransitionNode_1 = AnimStateTransitionNode() {
    blendMode=Custom
    crossfadeDuration=0.5
    customBlendCurve=/Game/Accessories/FORT_Backpacks/M_MED_Demon_Pack/Animation/JumpTransitionBlendOutCurve
   }
   AnimStateTransitionNode_11 = AnimStateTransitionNode()
   AnimStateTransitionNode_12 = AnimStateTransitionNode()
   AnimStateTransitionNode_13 = AnimStateTransitionNode()
   AnimStateTransitionNode_15 = AnimStateTransitionNode()
   AnimStateTransitionNode_2 = AnimStateTransitionNode()
   AnimStateTransitionNode_3 = AnimStateTransitionNode() {blendMode=Custom, customBlendCurve=&JumpTransitionBlendOutCurve}
   AnimStateTransitionNode_4 = AnimStateTransitionNode() {blendMode=Custom, customBlendCurve=&JumpTransitionBlendCurve}
   AnimStateTransitionNode_5 = AnimStateTransitionNode()
   AnimStateTransitionNode_6 = AnimStateTransitionNode() {crossfadeDuration=0.25}
   AnimStateTransitionNode_7 = AnimStateTransitionNode() {crossfadeDuration=0.25}
  ]
  links [
   {
    dst=Crouching
    rule=AnimGraph.AnimGraphNode_StateMachine_0.wingsPosition.AnimStateTransitionNode_12.AnimationTransitionGraph_0
    src=Base
    via=AnimStateTransitionNode_12
   }
   {
    dst=Falling
    rule=AnimGraph.AnimGraphNode_StateMachine_0.wingsPosition.AnimStateTransitionNode_4.AnimationTransitionGraph_0
    src=Base
    via=AnimStateTransitionNode_4
   }
   {
    dst=Jumping
    rule=AnimGraph.AnimGraphNode_StateMachine_0.wingsPosition.AnimStateTransitionNode_0.AnimationTransitionGraph_0
    src=Base
    via=AnimStateTransitionNode_0
   }
   {
    dst=SkydiveMode
    rule=AnimGraph.AnimGraphNode_StateMachine_0.wingsPosition.AnimStateTransitionNode_15.AnimationTransitionGraph_0
    src=Base
    via=AnimStateTransitionNode_15
   }
   {
    dst=Targeting
    rule=AnimGraph.AnimGraphNode_StateMachine_0.wingsPosition.AnimStateTransitionNode_6.AnimationTransitionGraph_0
    src=Base
    via=AnimStateTransitionNode_6
   }
   {
    dst=Base
    rule=AnimGraph.AnimGraphNode_StateMachine_0.wingsPosition.AnimStateTransitionNode_13.AnimationTransitionGraph_0
    src=Crouching
    via=AnimStateTransitionNode_13
   }
   {
    dst=Base
    rule=AnimGraph.AnimGraphNode_StateMachine_0.wingsPosition.AnimStateTransitionNode_3.AnimationTransitionGraph_0
    src=Falling
    via=AnimStateTransitionNode_3
   }
   {
    dst=Base
    rule=AnimGraph.AnimGraphNode_StateMachine_0.wingsPosition.AnimStateTransitionNode_1.AnimationTransitionGraph_0
    src=Jumping
    via=AnimStateTransitionNode_1
   }
   {
    dst=Falling
    rule=AnimGraph.AnimGraphNode_StateMachine_0.wingsPosition.AnimStateTransitionNode_2.AnimationTransitionGraph_0
    src=Jumping
    via=AnimStateTransitionNode_2
   }
   {
    dst=SkydiveMode
    rule=AnimGraph.AnimGraphNode_StateMachine_0.wingsPosition.AnimStateTransitionNode_5.AnimationTransitionGraph_0
    src=Jumping
    via=AnimStateTransitionNode_5
   }
   {
    dst=Base
    rule=AnimGraph.AnimGraphNode_StateMachine_0.wingsPosition.AnimStateTransitionNode_11.AnimationTransitionGraph_0
    src=SkydiveMode
    via=AnimStateTransitionNode_11
   }
   {
    dst=Base
    rule=AnimGraph.AnimGraphNode_StateMachine_0.wingsPosition.AnimStateTransitionNode_7.AnimationTransitionGraph_0
    src=Targeting
    via=AnimStateTransitionNode_7
   }
  ]
 }
 Base {
  path=AnimGraph.AnimGraphNode_StateMachine_0.wingsPosition.AnimStateNode_0.Base
  nodes [
   n1 = Wing_Movement_Base_Anim_PoseAsset : Base!(PoseWeight:float=1) -> (Pose:Pose Link Structure) {
    is=title AnimGraphNode_PoseByName
    poseAsset=&/Wing_Movement_Base_Anim_PoseAsset
    poseName=Base
   }
   Output Animation Pose(Result=@n1)
  ]
 }
 Jumping {
  path=AnimGraph.AnimGraphNode_StateMachine_0.wingsPosition.AnimStateNode_1.Jumping
  nodes [
   n1 = Wing_Movement_Base_Anim_PoseAsset : Open!(PoseWeight:float=1) -> (Pose:Pose Link Structure) {
    is=title AnimGraphNode_PoseByName
    poseAsset=&Wing_Movement_Base_Anim_PoseAsset
    poseName=Open
   }
   Output Animation Pose(Result=@n1)
  ]
 }
 Falling {
  path=AnimGraph.AnimGraphNode_StateMachine_0.wingsPosition.AnimStateNode_2.Falling
  nodes [
   n1 = Wing_Movement_Base_Anim_PoseAsset : Open(PoseWeight=1) {
    poseAsset=&Wing_Movement_Base_Anim_PoseAsset
    poseName=Open
   }
   Output Animation Pose(Result=@n1)
  ]
 }
 Targeting {
  path=AnimGraph.AnimGraphNode_StateMachine_0.wingsPosition.AnimStateNode_3.Targeting
  nodes [
   n1 = Wing_Movement_Base_Anim_PoseAsset : Close!(PoseWeight:float=1) -> (Pose:Pose Link Structure) {
    is=title AnimGraphNode_PoseByName
    poseAsset=&Wing_Movement_Base_Anim_PoseAsset
    poseName=Close
   }
   Output Animation Pose(Result=@n1)
  ]
 }
 SkydiveMode {
  path=AnimGraph.AnimGraphNode_StateMachine_0.wingsPosition.AnimStateNode_5.SkydiveMode
  nodes [
   AnimGraphNode_StateMachine_0 = Skydiving State() {is=title AnimGraphNode_StateMachine}
   Output Animation Pose(Result=@AnimGraphNode_StateMachine_0)
  ]
 }
 Skydiving State {
  entry=Parachuting
  path=AnimGraph.AnimGraphNode_StateMachine_0.wingsPosition.AnimStateNode_5.SkydiveMode.AnimGraphNode_StateMachine_0.Skydiving State
  ids=id
  nodes [
   AnimStateNode_1 = Diving() {is=title AnimStateNode}
   AnimStateNode_2 = Floating() {is=title AnimStateNode}
   AnimStateNode_4 = Parachuting() {is=title AnimStateNode}
   AnimStateTransitionNode_1 = AnimStateTransitionNode() {crossfadeDuration=0.5}
   AnimStateTransitionNode_10 = AnimStateTransitionNode()
   AnimStateTransitionNode_12 = AnimStateTransitionNode()
   AnimStateTransitionNode_13 = AnimStateTransitionNode()
   AnimStateTransitionNode_3 = AnimStateTransitionNode()
   AnimStateTransitionNode_5 = AnimStateTransitionNode() {crossfadeDuration=0.5}
  ]
  links [
   {
    dst=Floating
    rule=AnimGraph.AnimGraphNode_StateMachine_0.wingsPosition.AnimStateNode_5.SkydiveMode.AnimGraphNode_StateMachine_0.Skydiving State.AnimStateTransitionNode_1.AnimationTransitionGraph_0
    src=Diving
    via=AnimStateTransitionNode_1
   }
   {
    dst=Parachuting
    rule=AnimGraph.AnimGraphNode_StateMachine_0.wingsPosition.AnimStateNode_5.SkydiveMode.AnimGraphNode_StateMachine_0.Skydiving State.AnimStateTransitionNode_10.AnimationTransitionGraph_1
    src=Diving
    via=AnimStateTransitionNode_10
   }
   {
    dst=Diving
    rule=AnimGraph.AnimGraphNode_StateMachine_0.wingsPosition.AnimStateNode_5.SkydiveMode.AnimGraphNode_StateMachine_0.Skydiving State.AnimStateTransitionNode_5.AnimationTransitionGraph_0
    src=Floating
    via=AnimStateTransitionNode_5
   }
   {
    dst=Parachuting
    rule=AnimGraph.AnimGraphNode_StateMachine_0.wingsPosition.AnimStateNode_5.SkydiveMode.AnimGraphNode_StateMachine_0.Skydiving State.AnimStateTransitionNode_3.AnimationTransitionGraph_1
    src=Floating
    via=AnimStateTransitionNode_3
   }
   {
    dst=Diving
    rule=AnimGraph.AnimGraphNode_StateMachine_0.wingsPosition.AnimStateNode_5.SkydiveMode.AnimGraphNode_StateMachine_0.Skydiving State.AnimStateTransitionNode_13.AnimationTransitionGraph_1
    src=Parachuting
    via=AnimStateTransitionNode_13
   }
   {
    dst=Floating
    rule=AnimGraph.AnimGraphNode_StateMachine_0.wingsPosition.AnimStateNode_5.SkydiveMode.AnimGraphNode_StateMachine_0.Skydiving State.AnimStateTransitionNode_12.AnimationTransitionGraph_1
    src=Parachuting
    via=AnimStateTransitionNode_12
   }
  ]
 }
 Diving {
  path=AnimGraph.AnimGraphNode_StateMachine_0.wingsPosition.AnimStateNode_5.SkydiveMode.AnimGraphNode_StateMachine_0.Skydiving State.AnimStateNode_1.Diving
  nodes [
   n1 = Wing_Movement_Base_Anim_PoseAsset : Close(PoseWeight=1) {
    poseAsset=&Wing_Movement_Base_Anim_PoseAsset
    poseName=Close
   }
   Output Animation Pose(Result=@n1)
  ]
 }
 Floating {
  path=AnimGraph.AnimGraphNode_StateMachine_0.wingsPosition.AnimStateNode_5.SkydiveMode.AnimGraphNode_StateMachine_0.Skydiving State.AnimStateNode_2.Floating
  nodes [
   n1 = Wing_Movement_Base_Anim_PoseAsset : Close(PoseWeight=0.25) {
    poseAsset=&Wing_Movement_Base_Anim_PoseAsset
    poseName=Close
   }
   Output Animation Pose(Result=@n1)
  ]
 }
 Parachuting {
  path=AnimGraph.AnimGraphNode_StateMachine_0.wingsPosition.AnimStateNode_5.SkydiveMode.AnimGraphNode_StateMachine_0.Skydiving State.AnimStateNode_4.Parachuting
  nodes [
   n1 = Wing_Movement_Base_Anim_PoseAsset : Open(PoseWeight=0.8) {
    poseAsset=&Wing_Movement_Base_Anim_PoseAsset
    poseName=Open
   }
   Output Animation Pose(Result=@n1)
  ]
 }
 AnimationTransitionGraph_0 {
  path=AnimGraph.AnimGraphNode_StateMachine_0.wingsPosition.AnimStateNode_5.SkydiveMode.AnimGraphNode_StateMachine_0.Skydiving State.AnimStateTransitionNode_1.AnimationTransitionGraph_0
  nodes [Result(bCanEnterTransition=NOT Boolean(A=Get bIsSkydiveDiveMode()))]
 }
 AnimationTransitionGraph_1 {
  path=AnimGraph.AnimGraphNode_StateMachine_0.wingsPosition.AnimStateNode_5.SkydiveMode.AnimGraphNode_StateMachine_0.Skydiving State.AnimStateTransitionNode_10.AnimationTransitionGraph_1
  nodes [Result(bCanEnterTransition=Get bIsParachuteOpened())]
 }
 AnimationTransitionGraph_1 {
  path=AnimGraph.AnimGraphNode_StateMachine_0.wingsPosition.AnimStateNode_5.SkydiveMode.AnimGraphNode_StateMachine_0.Skydiving State.AnimStateTransitionNode_12.AnimationTransitionGraph_1
  nodes [Result(bCanEnterTransition=NOT Boolean(A=Get bIsParachuteOpened()))]
 }
 AnimationTransitionGraph_1 {
  path=AnimGraph.AnimGraphNode_StateMachine_0.wingsPosition.AnimStateNode_5.SkydiveMode.AnimGraphNode_StateMachine_0.Skydiving State.AnimStateTransitionNode_13.AnimationTransitionGraph_1
  nodes [Result(bCanEnterTransition=NOT Boolean(A=Get bIsParachuteOpened()))]
 }
 AnimationTransitionGraph_1 {
  path=AnimGraph.AnimGraphNode_StateMachine_0.wingsPosition.AnimStateNode_5.SkydiveMode.AnimGraphNode_StateMachine_0.Skydiving State.AnimStateTransitionNode_3.AnimationTransitionGraph_1
  nodes [Result(bCanEnterTransition=Get bIsParachuteOpened())]
 }
 AnimationTransitionGraph_0 {
  path=AnimGraph.AnimGraphNode_StateMachine_0.wingsPosition.AnimStateNode_5.SkydiveMode.AnimGraphNode_StateMachine_0.Skydiving State.AnimStateTransitionNode_5.AnimationTransitionGraph_0
  nodes [Result(bCanEnterTransition=Get bIsSkydiveDiveMode())]
 }
 Crouching {
  path=AnimGraph.AnimGraphNode_StateMachine_0.wingsPosition.AnimStateNode_9.Crouching
  nodes [
   n1 = Wing_Movement_Base_Anim_PoseAsset : Close(PoseWeight=0.8) {
    poseAsset=&Wing_Movement_Base_Anim_PoseAsset
    poseName=Close
   }
   Output Animation Pose(Result=@n1)
  ]
 }
 AnimationTransitionGraph_0 {
  path=AnimGraph.AnimGraphNode_StateMachine_0.wingsPosition.AnimStateTransitionNode_0.AnimationTransitionGraph_0
  nodes [Result(bCanEnterTransition=Get bIsJumping())]
 }
 AnimationTransitionGraph_0 {
  path=AnimGraph.AnimGraphNode_StateMachine_0.wingsPosition.AnimStateTransitionNode_1.AnimationTransitionGraph_0
  nodes [
   Result(bCanEnterTransition=NOT Boolean(A=Get bIsJumping()))
   Get bShouldPredictLanding() -> (bShouldPredictLanding:bool)
   Get bLandingPredicted() -> (bLandingPredicted:bool)
   Get bDebugLandPrediction() -> (bDebugLandPrediction:bool)
   Get bIsFalling()
  ]
 }
 AnimationTransitionGraph_0 {
  path=AnimGraph.AnimGraphNode_StateMachine_0.wingsPosition.AnimStateTransitionNode_11.AnimationTransitionGraph_0
  nodes [Result(bCanEnterTransition=NOT Boolean(A=Get bIsSkyDiving()))]
 }
 AnimationTransitionGraph_0 {
  path=AnimGraph.AnimGraphNode_StateMachine_0.wingsPosition.AnimStateTransitionNode_12.AnimationTransitionGraph_0
  nodes [Result(bCanEnterTransition=Get bIsCrouching())]
 }
 AnimationTransitionGraph_0 {
  path=AnimGraph.AnimGraphNode_StateMachine_0.wingsPosition.AnimStateTransitionNode_13.AnimationTransitionGraph_0
  nodes [Result(bCanEnterTransition=NOT Boolean(A=Get bIsCrouching()))]
 }
 AnimationTransitionGraph_0 {
  path=AnimGraph.AnimGraphNode_StateMachine_0.wingsPosition.AnimStateTransitionNode_15.AnimationTransitionGraph_0
  nodes [Result(bCanEnterTransition=Get bIsSkyDiving())]
 }
 AnimationTransitionGraph_0 {
  path=AnimGraph.AnimGraphNode_StateMachine_0.wingsPosition.AnimStateTransitionNode_2.AnimationTransitionGraph_0
  nodes [Result(bCanEnterTransition=Get bIsFalling())]
 }
 AnimationTransitionGraph_0 {
  path=AnimGraph.AnimGraphNode_StateMachine_0.wingsPosition.AnimStateTransitionNode_3.AnimationTransitionGraph_0
  nodes [Result(bCanEnterTransition=NOT Boolean(A=Get bIsFalling()))]
 }
 AnimationTransitionGraph_0 {
  path=AnimGraph.AnimGraphNode_StateMachine_0.wingsPosition.AnimStateTransitionNode_4.AnimationTransitionGraph_0
  nodes [Result(bCanEnterTransition=Get bIsFalling())]
 }
 AnimationTransitionGraph_0 {
  path=AnimGraph.AnimGraphNode_StateMachine_0.wingsPosition.AnimStateTransitionNode_5.AnimationTransitionGraph_0
  nodes [Result(bCanEnterTransition=Get bIsSkyDiving())]
 }
 AnimationTransitionGraph_0 {
  path=AnimGraph.AnimGraphNode_StateMachine_0.wingsPosition.AnimStateTransitionNode_6.AnimationTransitionGraph_0
  nodes [
   Result(bCanEnterTransition=Get bIsTargeting())
   Get is Targeting() -> (is Targeting:bool)
  ]
 }
 AnimationTransitionGraph_0 {
  path=AnimGraph.AnimGraphNode_StateMachine_0.wingsPosition.AnimStateTransitionNode_7.AnimationTransitionGraph_0
  nodes [Result(bCanEnterTransition=NOT Boolean(A=Get bIsTargeting()))]
 }
 EventGraph {nodes [
   Event BlueprintUpdateAnimation()
   > n1 = Sequence()
   n2 = TryGetPawnOwner()
   n3 = GetVelocity!(self:Actor Object Reference=@n2) -> (ReturnValue:Vector)
   n4 = Get Actor Rotation!(self:Actor Object Reference=@n2) -> (ReturnValue:Rotator) {is=title K2Node_CallFunction}
   n5 = BreakVector(InVec=vector * vector(A:Vector=Get Rotation X Vector!(InRot:Rotator=@n4) -> (ReturnValue:Vector) {is=title K2Node_CallFunction}, B:Vector=@n3) -> (ReturnValue:Vector) {is=title K2Node_CommutativeAssociativeBinaryOperator})
   n6 = float + float(A=@n5, B=@n5.Y) {is=title K2Node_CommutativeAssociativeBinaryOperator}
   n7 = "Not Equal (Float)"!(A:double=BreakVector(InVec=@n3).Z, B:double=0) -> (ReturnValue:bool) {is=title K2Node_CallFunction}
   > Set SprintSpeedRemap(SprintSpeedRemap:double=float * float(A=MapRangeClamped(Value=@n6, InRangeA=0, InRangeB=550), B=SelectFloat(A=0, B=1, bPickA=@n7))) -> (Output_Get:double)
   n8 = GetDeltaTransformFromRefPose(self=GetOwnerSkeletalMeshComponent(), BoneName=spine_03, BaseName=spine_01)
   n9 = "Absolute (Float)"(A=BreakRotator(InRot=BreakTransform(InTransform=@n8).Rotation))
   Set Backpack Rotate(execute=@n1.then_1, Backpack Rotate:Rotator=MakeRotator(Yaw=MapRangeClamped(Value=@n9, InRangeA=3, InRangeB=15, OutRangeB=-20))) -> (Output_Get:Rotator)
  ]}
]
variables [
 Backpack Rotate {type=Rotator}
 SprintSpeedRemap {type=number}
]
